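; Rising lava floor Blueprint for an Athena playlist, handling the lava's cosmetic effects.
; A repeating cosmetic-only timer aligns an FX component to the local view target's XY at lava-surface height, feeds that position to smoke materials, and writes actor height to the minimap parameter collection.
; When movement starts, it plays a sound and optionally fades in world smoke once through a timeline.
; Smoke stays off on Switch, and the height-event handler performs no effective work.
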(event EventBeginPlay
  (Networking|CanExecuteCosmeticEvents
    (:True
      (Utilities|FlowControl|Sequence
        (:then_1
          ;[1 1 1] Turns off all lava smoke meshes if on the Switch, or if we're going to fade them in once the lava starts rising.
          (if (or (TouchPlatform|IsNintendoSwitch) (Variables|Default|GetbFadeInSmokeOnStartMoving))
            (CallFunction|ToggleSmoke false)))
        (:then_2
          (Utilities|Time|SetTimerbyEvent (EventDispatchers|CreateEvent :!event Custom|Update%20Player%20Aligned%20Surface%20FX) (Variables|Default|GetFXUpdateFrequency) true))))))

(event Custom|Update%20Player%20Aligned%20Surface%20FX
  (bind _sm_lavaground_blendfx (Variables|Default|GetSM_LavaGround_BlendFX))
  (bind (_returnvalue_x _returnvalue_y _returnvalue_z) (Math|Vector|BreakVector (Transformation|GetWorldLocation _sm_lavaground_blendfx)))
  ;[1 1 1] Move the "PlayerAlignedLocation" scene comp to the correct location
  (Utilities|FlowControl|Sequence
    (:then_0
      (bind _output_get (Variables|Default|SetPlayerLocation (CallFunction|GetPlayerLocation)))
      (Transformation|SetWorldLocation (Variables|Default|GetPlayerAlignedLocation) (Math|Vector|MakeVector (.x _output_get) (.y _output_get) _returnvalue_z) false false)
      (Rendering|Material|SetVectorParameterValueonMaterials (Variables|Default|GetSM_Lava_Scatter_SmokePlumes_World) "LavaPosition" (Math|Vector|MakeVector (.x _output_get) (.y _output_get) _returnvalue_z)))
    (:then_1
      (Rendering|Material|SetScalarParameterValue "/Game/Athena/HUD/MiniMap/MPC_Minimap.MPC_Minimap" "Height" (.z (Transformation|GetActorLocation)) :!class /Script/Engine.KismetMaterialLibrary))))

(event EventOnHeightEvent
  (params (Height :type Float%20%28single-precision%29) (HeightEventIndex :type Integer))
  ;[1 1 1] Set global ambient audio overrides at certain height thresholds. "Event Heights" array on this BP.
  (Utilities|IsValid 0))

(event EventOnStartMoving
  (params (CurrentStep :type Integer))
  ;[1 1 1] Lava starts to move
  (Audio|PlaySound2D (Variables|Default|GetSoundOnLavaStartMoving))
  (if (and (Variables|Default|GetbFadeInSmokeOnStartMoving) (not (TouchPlatform|IsNintendoSwitch)))
    (CallFunction|ToggleSmoke true)
    (Variables|Default|SetbFadeInSmokeOnStartMoving false)
    ;[0 0 0] Fade in world smoke
    (bind (_direction _opacity_alpha) (Timeline|FadeInWorldSmoke PlayFromStart)
      (:Update
        (Rendering|Material|SetScalarParameterValueonMaterials (Variables|Default|GetSM_Lava_Scatter_SmokePlumes_World) "Opacity" (Math|Float|Lerp 0.0 (Variables|Default|GetMaxWorldSmokeOpacity) _opacity_alpha))))))

(fn GetPlayerLocation
  (bind _returnvalue (Fortnite|GetLocalFortPlayerControllers true true))
  (bind _output (Utilities|Array|Get(acopy) _returnvalue))
  (bind _returnvalue_2 (Pawn|GetViewTarget _output))
  (Utilities|IsValid _output
    (:Is%20Valid
      (Utilities|IsValid _returnvalue_2
        (:Is%20Valid
          (return (Transformation|GetActorLocation _returnvalue_2)))
        (:Is%20Not%20Valid
          _)))
    (:Is%20Not%20Valid
      _))
  (return "0, 0, 0"))

(fn ToggleSmoke
  (params (bActive :type Boolean))
  (Components|Activation|SetActive (Variables|Default|GetSM_Lava_Scatter_SmokePlumes_World) bActive false)
  (Rendering|SetVisibility (Variables|Default|GetSM_Lava_Scatter_SmokePlumes_World) bActive false))

(fn ConstructionScript)
